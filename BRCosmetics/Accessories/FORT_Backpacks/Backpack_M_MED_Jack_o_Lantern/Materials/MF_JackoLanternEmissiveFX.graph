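# Material function for the Jack-o-Lantern backpack that adds a flickering, parallax interior glow to the incoming material attributes inside a face mask.
# Bump-offset mask samples, optionally blended with a nose texture, shape the glow.
# A view-facing term minus a time-driven curve-atlas value flickers it, and time-of-day strength feeds its brightness.
# The mask blends emissive, base color, roughness and specular; the panning camera-vector noise chain is unconnected to the output.
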MaterialFunction /BRCosmetics/Accessories/FORT_Backpacks/Backpack_M_MED_Jack_o_Lantern/Materials/MF_JackoLanternEmissiveFX
graph {nodes [
  n1 = CameraVectorWS()
  n2 = ScalarParameter() {defaultValue=8, parameterName=Tiling}
  n3 = CameraVectorWS()
  n4 = Add(
   A=Multiply(A=@n3, B=ScalarParameter() {defaultValue=4, parameterName=Tiling})
   B=Multiply(A=Time(), B=VectorParameter() {parameterName=DistortionPanDirectionAndSpeed, defaultValue=<rgba 0 -0.2 -0.45 1>}.RGB)
  )
  n5 = TextureSample(UVs=ComponentMask(None=@n4) {b=true, r=true}) {samplerSource=SSM_Wrap_WorldGroupSettings, texture=/Game/Characters/Player/Male/Medium/Bodies/M_MED_Viper/Textures/Glitch/T_SoftNoise}
  n6 = TextureSample(UVs=ComponentMask(None=@n4) {b=true, g=true}) {samplerSource=SSM_Wrap_WorldGroupSettings, texture=&T_SoftNoise}
  n7 = DotProduct(A=@n3, B=Constant3Vector() {constant=<rgba 1 0 0 0>})
  n8 = ScalarParameter() {defaultValue=12, parameterName=FadeContrast}
  n9 = LinearInterpolate(
   A=@n5.RGB
   B=@n6.RGB
   Alpha=MaterialFunctionCall(In=Abs(None=@n7), Power=@n8) {materialFunction=&SCurve}.Result
  )
  n10 = TextureSample(UVs=ComponentMask(None=@n4) {g=true, r=true}) {samplerSource=SSM_Wrap_WorldGroupSettings, texture=&T_SoftNoise}
  n11 = Constant3Vector() {constant=<rgba 0 0 1 0>}
  n12 = MaterialFunctionCall(In=Abs(None=DotProduct(A=@n3, B=@n11)), Power=@n8) {materialFunction=&SCurve}
  n13 = Subtract(A=Multiply(A=LinearInterpolate(A=@n9, B=@n10.RGB, Alpha=@n12.Result)) {constB=2})
  n14 = ScalarParameter() {defaultValue=0.05, parameterName=DistortionStrength}
  n15 = Add(
   A=Add(A=Multiply(A=@n1, B=@n2), B=Multiply(A=@n13, B=@n14))
   B=Multiply(A=Time(), B=VectorParameter() {parameterName=PanDirectionAndSpeed, defaultValue=<rgba 0 0.25 0.5 1>}.RGB)
  )
  n16 = TextureSample(UVs=ComponentMask(None=@n15) {b=true, r=true}) {samplerSource=SSM_Wrap_WorldGroupSettings, texture=&T_SoftNoise}
  n17 = TextureSample(UVs=ComponentMask(None=@n15) {b=true, g=true}) {samplerSource=SSM_Wrap_WorldGroupSettings, texture=&T_SoftNoise}
  n18 = DotProduct(A=@n1, B=Constant3Vector() {constant=<rgba 1 0 0 0>})
  n19 = ScalarParameter() {defaultValue=12, parameterName=FadeContrast}
  n20 = LinearInterpolate(
   A=@n16.RGB
   B=@n17.RGB
   Alpha=MaterialFunctionCall(In=Abs(None=@n18), Power=@n19) {materialFunction=&SCurve}.Result
  )
  n21 = TextureSample(UVs=ComponentMask(None=@n15) {g=true, r=true}) {samplerSource=SSM_Wrap_WorldGroupSettings, texture=&T_SoftNoise}
  n22 = Constant3Vector() {constant=<rgba 0 0 1 0>}
  n23 = MaterialFunctionCall(In=Abs(None=DotProduct(A=@n1, B=@n22)), Power=@n19) {materialFunction=&SCurve}
  Clamp(None=ChannelMaskParameter(None=LinearInterpolate(A=@n20, B=@n21.RGB, Alpha=@n23.Result)) {parameterName=Param_1}.R) {minDefault=0.25}
  n24 = Time()
  ConstantBiasScale(None=Sine(None=Multiply(A=@n24) {constB=0.2}))
  n25 = GetMaterialAttributes(None=FunctionInput() {inputName=Material, inputType=FunctionInput_MaterialAttributes}) {attributeGetTypes [
    #1
    #2
    #3
    #4
   ]}
  n26 = TextureCoordinate()
  n27 = TextureSampleParameter2D(UVs=@n26) {
   parameterName=FX_Mask
   samplerType=SAMPLERTYPE_Masks
   texture=/BRCosmetics/Accessories/FORT_Backpacks/Backpack_M_MED_Jack_o_Lantern/Textures/Red/T_Jack_o_lantern_Red_FaceTop_FX
  }
  n28 = TextureSampleParameter2D(UVs=@n26) {
   parameterName=Nose_FX_Mask
   samplerType=SAMPLERTYPE_Masks
   texture=&T_Jack_o_lantern_Red_FaceTop_FX
  }
  n29 = FunctionInput() {inputName=NoseMask, inputType=FunctionInput_Scalar}
  n30 = StaticSwitchParameter(True=LinearInterpolate(A=@n27.RGBA, B=@n28.RGBA, Alpha=@n29), False=@n27.RGBA) {parameterName=UsetwoTextures}
  n31 = BumpOffset(
   Height=ChannelMaskParameter(None=@n30) {maskChannel=Alpha, parameterName=HeightGradient_Channel, defaultValue=<rgba 0 0 0 1>}.R
   HeightRatioInput=LinearInterpolate(
    A=ScalarParameter() {defaultValue=0.1, parameterName=HeightRatio_input}
    B=ScalarParameter() {defaultValue=0.025, parameterName=NoiseHeightRatioInput}
    Alpha=@n29
   )
  ) {heightRatio=0.1}
  n32 = ScalarParameter() {defaultValue=1, parameterName=MipLevel}
  n33 = TextureSampleParameter2D!(UVs=@n31, Bias=@n32, Apply View MipBias) {
   mipValueMode=TMVM_MipBias
   parameterName=FX_Mask
   samplerType=SAMPLERTYPE_Masks
   texture=&T_Jack_o_lantern_Red_FaceTop_FX
  }
  n34 = TextureSampleParameter2D(UVs=@n31, Bias=@n32) {
   mipValueMode=TMVM_MipBias
   parameterName=Nose_FX_Mask
   samplerType=SAMPLERTYPE_Masks
   texture=&T_Jack_o_lantern_Red_FaceTop_FX
  }
  n35 = ChannelMaskParameter(None=StaticSwitchParameter(True=LinearInterpolate(A=@n33.RGBA, B=@n34.RGBA, Alpha=@n29), False=@n33.RGBA) {parameterName=UsetwoTextures}) {maskChannel=Alpha, defaultValue=<rgba 0 0 0 1>}
  n36 = Time()
  n37 = ScalarParameter() {defaultValue=0.1, parameterName=FlickerSpeed}
  n38 = ConstantBiasScale(None=Sine(None=Multiply(A=@n36, B=@n37)))
  n39 = CurveAtlasRowParameter(CurveTime=@n38) {
   atlas=&AthenaCurveAtlas01
   curve=/Game/Accessories/FORT_Backpacks/Backpack_M_MED_Jack_o_lantern/Textures/JackolanternCurve
   defaultValue=307
   parameterName=FlickerCurve
  }
  n40 = Subtract(A=DotProduct(A=CameraVectorWS(), B=PixelNormalWS()), B=@n39.R)
  n41 = Multiply(A=Power(Base=Clamp(None=@n40))) {constB=2}
  n42 = Multiply(A=Power(Base=OneMinus(None=@n35.R)), B=@n41)
  n43 = ScalarParameter() {defaultValue=3, parameterName=EmissiveStrength}
  n44 = ScalarParameter() {defaultValue=2, parameterName=Emissive_Strength_Night}
  n45 = Multiply(
   A=Multiply(A=@n42, B=VectorParameter() {parameterName=EmissiveColor, defaultValue=<rgba 0.807 0 0.0222 1>}.RGB)
   B=LinearInterpolate(
    A=ScalarParameter() {defaultValue=1, parameterName=Parallax_EmissiveStrength}
    B=MaterialFunctionCall(Color_Morning=@n43, Color_Day=@n43, Color_Evening=@n43, Color_Night=@n44) {materialFunction=&Color_Time_of_Day}.Result
    Alpha=@n42
   ) {constA=1}
  )
  n46 = Multiply(A=NamedRerouteUsage(), B=Multiply(A=ScalarParameter() {defaultValue=0.1, parameterName=Parallax_Emissive_Darkening}, B=@n41))
  n47 = ChannelMaskParameter(None=@n30) {maskChannel=Alpha, parameterName="EmisiveMask_Channe;", defaultValue=<rgba 0 0 0 1>}
  n48 = Clamp(None=SphereMask(A=@n47.R, B=Constant(), Radius=Constant() {r=0.25}, Hardness=Constant()))
  n49 = Multiply(A=NamedRerouteUsage(), B=ScalarParameter() {defaultValue=0.15, parameterName=BaseColorDarkening})
  n50 = SetMaterialAttributes!(
   MaterialAttributes=@n25.MaterialAttributes
   EmissiveColor=LinearInterpolate(A=@n25.EmissiveColor, B=LinearInterpolate(A=@n46, B=@n45, Alpha=@n42), Alpha=@n48)
   BaseColor=LinearInterpolate(A=@n25.BaseColor, B=LinearInterpolate(A=@n49, B=@n45, Alpha=@n42), Alpha=@n48)
   Roughness=LinearInterpolate(A=@n25.Roughness, Alpha=@n48)
   Specular=LinearInterpolate(A=@n25.Specular, Alpha=@n48) {constB=0}
  ) {attributeSetTypes [
    #1
    #2
    #4
    #3
   ]}
  FunctionOutput(None=@n50)
  Multiply(A=@n41, B=Clamp(None=@n38) {minDefault=0.5})
  NamedRerouteDeclaration(None=@n25.BaseColor) {name=BaseColor, nodeColor=<rgba 1 0 0.7266 1>}
  OneMinus(None=@n47.R)
  Power(Base=@n47.R)
 ]}
