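# Base material function for the Jack-o'-Lantern backpack cosmetic, outputting material attributes plus separate diffuse, normal, roughness, packed mask and NoseMask values.
# UsetwoTextures switches in a second Nose set for diffuse and normals, while a NoseMask is built from offset, center-scaled UVs.
# Packed spec/metal/roughness channels get optional metal power adjustment and a low-quality mobile roughness override.
# Nose blends for specular and M masks exist but feed no output.
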MaterialFunction /BRCosmetics/Accessories/FORT_Backpacks/Backpack_M_MED_Jack_o_Lantern/Materials/MF_BaseJack0Lantern
graph {nodes [
  n1 = FunctionInput(Preview=TextureCoordinate()) {bUsePreviewValueAsDefault=true, inputName=TexCoords, inputType=FunctionInput_Vector2}
  n2 = TextureSampleParameter2D(UVs=@n1) {group=Normal, parameterName=Normals, samplerType=SAMPLERTYPE_Normal, texture=/Engine/EngineMaterials/DefaultNormal}
  n3 = LinearInterpolate(
   A=@n2.RGB
   B=TextureSampleParameter2D(UVs=@n1) {group=Normal, parameterName=Nose_Normals, samplerType=SAMPLERTYPE_Normal, texture=&DefaultNormal}.RGB
   Alpha=Ceil(None=NamedRerouteUsage())
  )
  AppendVector(A=@n3, B=@n2.B)
  n4 = TextureSampleParameter2D(UVs=@n1) {group=Diffuse, parameterName=Diffuse, texture=/Engine/EngineMaterials/DefaultDiffuse}
  n5 = StaticSwitchParameter(
   True=LinearInterpolate(
    A=@n4.RGB
    B=TextureSampleParameter2D(UVs=@n1) {group=Diffuse, parameterName=Nose_Diffuse, texture=&DefaultDiffuse}.RGB
    Alpha=NamedRerouteUsage()
   )
   False=@n4.RGB
  ) {parameterName=UsetwoTextures}
  n6 = TextureSampleParameter2D(UVs=@n1) {
   desc="R = SpecMask G = Metallic B = Roughness"
   group=Specular
   parameterName=SpecularMasks
   samplerType=SAMPLERTYPE_Masks
   texture=&T_Fortnite_Default_S
  }
  n7 = ChannelMaskParameter(None=@n6.RGB) {maskChannel=Green, parameterName=Green, defaultValue=<rgba 0 1 0 0>}
  n8 = Saturate(None=Power(Base=@n7.R, Exp=ScalarParameter() {defaultValue=0.4, group=Masks, parameterName=MetalMaskPower}))
  n9 = ScalarParameter() {group=Roughness, parameterName=MobileRoughnessMin}
  n10 = ScalarParameter() {defaultValue=1, group=Roughness, parameterName=MobileRoughnessMax}
  n11 = ChannelMaskParameter(None=@n6.RGB) {maskChannel=Blue, parameterName=Blue, defaultValue=<rgba 0 0 1 0>}
  n12 = LinearInterpolate(
   A=ScalarParameter() {group=Roughness, parameterName=RoughnessMin}
   B=ScalarParameter() {defaultValue=1, group=Roughness, parameterName=RoughnessMax}
   Alpha=@n11.R
  )
  n13 = QualitySwitch(
   Default=@n12
   Low=StaticSwitchParameter(True=LinearInterpolate(A=@n9, B=@n10, Alpha=@n11.R), False=@n12) {parameterName=OverrideRoughnessOnMobile}
  )
  n14 = Constant()
  n15 = StaticSwitchParameter(True=Normalize(VectorInput=@n3), False=@n2.RGB) {parameterName=UsetwoTextures}
  FunctionOutput(None=MakeMaterialAttributes!(
    BaseColor=@n5
    Metallic=StaticSwitchParameter(True=@n8, False=@n7.R) {group=Masks, parameterName=useMetalMaskPowerAdjustment}
    Specular=ChannelMaskParameter(None=@n6.RGB) {parameterName=Red}.R
    Roughness=@n13
    Anisotropy
    EmissiveColor=@n14
    Opacity=Constant()
    OpacityMask
    Normal=@n15
    Tangent
    WorldPositionOffset
    SubsurfaceColor=@n14
    ClearCoat=@n14
    ClearCoatRoughness=@n14
    AmbientOcclusion
    Refraction
    CustomizedUVs_0
    CustomizedUVs_1
    CustomizedUVs_2
    CustomizedUVs_3
    CustomizedUVs_4
    CustomizedUVs_5
    CustomizedUVs_6
    CustomizedUVs_7
    PixelDepthOffset
    ShadingModel=ShadingModel()
    Displacement
   )) {outputName=Result_1}
  FunctionOutput(None=@n15) {outputName=Normal Map, sortPriority=2}
  FunctionOutput(None=@n13) {outputName=Roughness, sortPriority=8}
  FunctionOutput(None=@n6.RGB) {outputName=SpecularMask, sortPriority=3}
  n16 = TextureSampleParameter2D(UVs=@n1) {
   desc="Masks R=AO G= Fresnel B = Tmission/Skin"
   group=Masks
   parameterName=M
   samplerType=SAMPLERTYPE_Masks
   texture=/Engine/EngineMaterials/DefaultDiffuse_TC_Masks
  }
  FunctionOutput(None=@n16.RGB) {outputName=M, sortPriority=4}
  FunctionOutput(None=@n5) {outputName=DIffuse, sortPriority=1}
  n17 = ScalarParameter() {defaultValue=0.4, parameterName=NoseMask_offset_X}
  n18 = ScalarParameter() {defaultValue=0.4, parameterName=NoseMask_offset_Y}
  n19 = MaterialFunctionCall(UVs=Add(A=@n1, B=AppendVector(A=@n17, B=@n18)), Texture Scale=ScalarParameter() {defaultValue=0.2, parameterName=NoseMaskScale}) {materialFunction=&ScaleUVsByCenter}
  FunctionOutput(None=NamedRerouteDeclaration(None=ComponentMask(None=@n19.0-1 mask) {r=true}) {name=NoseMask, nodeColor=<rgba 1 0.6563 0 1>}) {outputName=NoseMask}
  MaterialFunctionCall!(Material, Normal) {materialFunction=/Engine/Functions/MaterialLayerFunctions/MatLayerBlend_BakedNormal}
  NamedRerouteDeclaration(None=@n11.R) {name=PreFX_RoughnessRawPaint, nodeColor=<rgba 1 0.5156 0 1>}
  n20 = TextureSampleParameter2D(UVs=@n1) {
   desc="Masks R=AO G= Fresnel B = Tmission/Skin"
   group=Masks
   parameterName=Nose_M
   samplerType=SAMPLERTYPE_Masks
   texture=&DefaultDiffuse_TC_Masks
  }
  StaticSwitchParameter(True=LinearInterpolate(A=@n16.RGB, B=@n20.RGB, Alpha=NamedRerouteUsage()), False=@n16.RGB) {parameterName=UsetwoTextures}
  n21 = TextureSampleParameter2D(UVs=@n1) {
   desc="R = SpecMask G = Metallic B = Roughness"
   group=Specular
   parameterName=Nose_SpecularMasks
   samplerType=SAMPLERTYPE_Masks
   texture=&T_Fortnite_Default_S
  }
  StaticSwitchParameter(True=LinearInterpolate(A=@n6.RGB, B=@n21.RGB, Alpha=NamedRerouteUsage()), False=@n6.RGB) {parameterName=UsetwoTextures}
 ]}
