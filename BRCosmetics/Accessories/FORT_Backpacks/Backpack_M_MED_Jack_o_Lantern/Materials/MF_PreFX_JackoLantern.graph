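# Material function that builds the pre-effects base material for the Jack-o-Lantern backpack cosmetic.
# MF_BaseJack0Lantern supplies material, normals and a packed AO/GMap/Skin mask to chained shared character GMap color, brightness balance, emissive and skin functions.
# A normal-map-driven Fresnel feeds the emissive step.
# It outputs the result plus masks, Fresnel, distance blend and pre-skinned local position, which uses a stage switch and VertexInterpolator to work in both shader stages.
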MaterialFunction /BRCosmetics/Accessories/FORT_Backpacks/Backpack_M_MED_Jack_o_Lantern/Materials/MF_PreFX_JackoLantern
graph {nodes [
  n1 = MaterialFunctionCall!(TexCoords=FunctionInput(Preview=TextureCoordinate()) {bUsePreviewValueAsDefault=true, inputName=TexCoords, inputType=FunctionInput_Vector2}) {materialFunction=&/MF_BaseJack0Lantern}
  n2 = MaterialFunctionCall(Float3=@n1.M) {desc="R = AO\r\nG = GMap\r\nB = Skin", materialFunction=&BreakOutFloat3Components}
  n3 = MaterialFunctionCall!(Material=@n1.Result_1, Diffuse=@n1.Result_1, GMap=@n2.G) {materialFunction=/Game/Packages/Fortress_SharedMaterials/Base_Material_Functions/BaseCharacterShaderFunctions/BaseFN_CharacterGMapColor}
  n4 = MaterialFunctionCall!(Material=@n3.Result) {materialFunction=/Game/Packages/Fortress_SharedMaterials/Base_Material_Functions/CharacterBrightnessandScale_Balance}
  n5 = Fresnel(Normal=Transform(None=@n1.Normal Map)) {baseReflectFraction=0, exponent=1}
  n6 = MaterialFunctionCall!(
   Material=MaterialFunctionCall!(Material=@n4.Result, Diffuse=@n3.DIffuse, BalanceDistance=@n4.Result, Normals=@n1.Normal Map, AO=@n2.R, Fresnel=@n5) {materialFunction=/Game/Packages/Fortress_SharedMaterials/Base_Material_Functions/BaseCharacterShaderFunctions/BaseFN_CharacterEmissive}.Result
   AO=@n2.R
   Skin Mask=@n2.B
  ) {materialFunction=/Game/Packages/Fortress_SharedMaterials/Base_Material_Functions/BaseCharacterShaderFunctions/BaseFN_CharacterSkin}
  FunctionOutput(None=@n6.Result) {outputName=Result_1}
  FunctionOutput(None=@n2.R) {outputName=AO}
  FunctionOutput(None=@n2.G) {outputName=Gmap}
  FunctionOutput(None=@n2.B) {outputName=Skin}
  FunctionOutput(None=@n5) {outputName=Fresnel}
  FunctionOutput(None=@n4.Distance Blend) {outputName=Distance Blend}
  n7 = PreSkinnedPosition()
  n8 = ShaderStageSwitch(PixelShader=VertexInterpolator(VS=@n7).PS, VertexShader=@n7)
  FunctionOutput(None=@n8) {outputName=Pre skinned Local Position}
  FunctionOutput(None=@n1.NoseMask) {outputName=NoseMask}
  NamedRerouteDeclaration(None=@n2.B) {name=PreFX_SkinMask, nodeColor=<rgba 1 0.5156 0 1>}
  NamedRerouteDeclaration(None=@n4.Distance Blend) {name=PreFX_Distance Blend, nodeColor=<rgba 1 0.5156 0 1>}
  NamedRerouteDeclaration(None=@n1.SpecularMask) {name=PreFX_SpecularMask, nodeColor=<rgba 1 0.5156 0 1>}
  NamedRerouteDeclaration(None=@n1.Roughness) {name=PreFX_Roughness, nodeColor=<rgba 1 0.5156 0 1>}
  NamedRerouteDeclaration(None=@n8) {name=Pre Skinned Local Postition, nodeColor=<rgba 1 0.5156 0 1>}
  NamedRerouteDeclaration(None=@n6.Result) {name=PreFX_Result, nodeColor=<rgba 1 0.5156 0 1>}
  NamedRerouteDeclaration(None=@n2.R) {name=PreFX_AO, nodeColor=<rgba 1 0.5156 0 1>}
  NamedRerouteDeclaration(None=@n2.G) {name=PreFX_Gmap, nodeColor=<rgba 1 0.5156 0 1>}
  NamedRerouteDeclaration(None=@n5) {name=PreFX_Fresnel, nodeColor=<rgba 1 0.5156 0 1>}
  Reroute()
  Saturate(None=DotProduct(A=PixelNormalWS(), B=CameraVectorWS()))
  SetMaterialAttributes!(MaterialAttributes, BaseColor) {attributeSetTypes [#1]}
 ]}
