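# Master material for the Jack-o-Lantern backpack cosmetic, built for skeletal meshes with morph targets and mesh deformers.
# MF_PreFX_JackoLantern supplies the base material attributes and a nose mask to MF_JackoLanternEmissiveFX.
# That result passes through the shared MF_TechArtOverrides_accessories function, so all surface logic lives in material functions.
# The shading model comes from the expression, using Default Lit or Subsurface Profile with a hero skin profile.
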Material /BRCosmetics/Accessories/FORT_Backpacks/Backpack_M_MED_Jack_o_Lantern/Materials/M_JackoLantern_Master
graph {
 properties {
  bUseMaterialAttributes=true
  bUsedWithMeshDeformer=true
  bUsedWithMorphTargets=true
  bUsedWithSkeletalMesh=true
  pixelDepthOffsetMode=PDOM_Legacy
  shadingModel=MSM_FromMaterialExpression
  subsurfaceProfile=&SS_HeroSkin_02
  usedShadingModels=Default Lit | Subsurface Profile
 }
 nodes [
  n1 = MaterialFunctionCall!(TexCoords) {materialFunction=&/MF_PreFX_JackoLantern}
  MaterialFunctionCall(In=MaterialFunctionCall!(NoseMask=@n1.NoseMask, Material=@n1.NoseMask) {materialFunction=&/MF_JackoLanternEmissiveFX}.Result) {materialFunction=&MF_TechArtOverrides_accessories}
 ]
}
