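# Physics asset for the green-topper Jack-o-Lantern backpack cosmetic, setting up physics for the pumpkin's top on the pack.
# A tiny capsule body sits on the Dyn_Pumpkin_Top bone, and a thin, wide box body sits on the pack bone.
# A single constraint joins the two with tight swing and twist limits of a few degrees, keeping any motion of the top to a slight wobble.
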PhysicsAsset /BRCosmetics/Accessories/FORT_Backpacks/Backpack_M_MED_Jack_o_Lantern/Meshes/M_MED_Jack_o_lantern_GreenTopper_Pack_Physics
bodies {
 Dyn_Pumpkin_Top [{shapeType=Capsule, length=1, radius=1, shapeName=Dyn_Pumpkin_Top_capsule, center=<-0.8633 -0.2265 0>, rotation=<rot 90 0 0>}]
 pack [{shapeType=Box, extentX=19, extentY=19, extentZ=0.5, shapeName=pack_box, center=<-0.0001 -9.6885 6.2936>}]
}
constraints [{bone1Name=Dyn_Pumpkin_Top, bone2Name=pack, swing1LimitDegrees=1, swing2LimitDegrees=5, twistLimitDegrees=1}]
