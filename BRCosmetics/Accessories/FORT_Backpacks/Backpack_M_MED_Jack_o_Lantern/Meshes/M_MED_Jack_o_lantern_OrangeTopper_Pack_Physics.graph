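# Physics asset for the orange topper variant of the Jack-o'-Lantern backpack cosmetic, covering the pumpkin top bones.
# It has two box bodies: a wide, thin slab on Dyn_Master_Pumpkin_Top and a small box on Dyn_Pumpkin_Top.
# A single constraint links the two bodies with swing and twist limit angles of only a few degrees.
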PhysicsAsset /BRCosmetics/Accessories/FORT_Backpacks/Backpack_M_MED_Jack_o_Lantern/Meshes/M_MED_Jack_o_lantern_OrangeTopper_Pack_Physics
bodies {
 Dyn_Master_Pumpkin_Top [{shapeType=Box, extentX=14.5, extentY=15, extentZ=1, shapeName=Dyn_Master_Pumpkin_Top_box, center=<0 1.3002 -2.8971>, rotation=<rot 0 -90 0>}]
 Dyn_Pumpkin_Top [{shapeType=Box, extentX=1, extentY=3, extentZ=3, shapeName=Dyn_Pumpkin_Top_box, center=<0.2 0 0>}]
}
constraints [{bone1Name=Dyn_Pumpkin_Top, bone2Name=Dyn_Master_Pumpkin_Top, swing1LimitDegrees=8, swing2LimitDegrees=3, twistLimitDegrees=1}]
