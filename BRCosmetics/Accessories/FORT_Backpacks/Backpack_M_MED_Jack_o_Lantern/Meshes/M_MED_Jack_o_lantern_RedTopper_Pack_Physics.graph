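# Physics asset for the RedTopper Jack-o-Lantern backpack cosmetic, defining capsule bodies and joints for its pumpkin hat bones.
# A large capsule covers the pack, with Dyn_Pumpkin_Top joined to it by a constraint with free swing and twist.
# The left and right hat sides and a two-link hat-top chain hang off Dyn_Pumpkin_Top with narrow angular limits and almost no twist.
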PhysicsAsset /BRCosmetics/Accessories/FORT_Backpacks/Backpack_M_MED_Jack_o_Lantern/Meshes/M_MED_Jack_o_lantern_RedTopper_Pack_Physics
bodies {
 Dyn_Pumpkin_Hat_Side_l [{
   shapeType=Capsule
   length=10.6781
   radius=3.7626
   shapeName=Dyn_Pumpkin_Hat_Side_l_capsule
   center=<3.5148 -2.8917 -1.8964>
   rotation=<rot -6.9481 -89.9995 12.3683>
  }]
 Dyn_Pumpkin_Hat_Side_r [{
   shapeType=Capsule
   length=9.3692
   radius=3.473
   shapeName=Dyn_Pumpkin_Hat_Side_r_capsule
   center=<3.8345 -2.6627 -1.97>
   rotation=<rot -2.4456 -89.9996 17.3125>
  }]
 Dyn_Pumpkin_Hat_Top1 [{shapeType=Capsule, length=2, radius=0.5, shapeName=Dyn_Pumpkin_Hat_Top1_capsule, center=<1 0 0>, rotation=<rot 90 0 0>}]
 Dyn_Pumpkin_Hat_Top2 [{shapeType=Capsule, length=0.5, radius=0.5, shapeName=Dyn_Pumpkin_Hat_Top2_capsule, center=<0.5 0 0>, rotation=<rot 90 0 0>}]
 Dyn_Pumpkin_Top [{shapeType=Capsule, length=0.5, radius=0.5, shapeName=Dyn_Pumpkin_Top_capsule}]
 pack [{shapeType=Capsule, length=3.25, radius=12.5, shapeName=pack_capsule, center=<-0.2358 -8.0925 -3.7673>, rotation=<rot 90 0 59.9991>}]
}
constraints [
 {
  bone1Name=Dyn_Pumpkin_Top
  bone2Name=pack
  swing1LimitDegrees=45
  swing1Motion=Free
  swing2LimitDegrees=45
  swing2Motion=Free
  twistLimitDegrees=45
  twistMotion=Free
 }
 {bone1Name=Dyn_Pumpkin_Hat_Side_r, bone2Name=Dyn_Pumpkin_Top, swing1LimitDegrees=30, swing2LimitDegrees=5, twistLimitDegrees=1}
 {bone1Name=Dyn_Pumpkin_Hat_Side_l, bone2Name=Dyn_Pumpkin_Top, swing1LimitDegrees=30, swing2LimitDegrees=5, twistLimitDegrees=1}
 {bone1Name=Dyn_Pumpkin_Hat_Top1, bone2Name=Dyn_Pumpkin_Top, swing1LimitDegrees=15, swing2LimitDegrees=10, twistLimitDegrees=1}
 {bone1Name=Dyn_Pumpkin_Hat_Top2, bone2Name=Dyn_Pumpkin_Hat_Top1, swing1LimitDegrees=10, swing2LimitDegrees=1, twistLimitDegrees=1}
]
