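# Animation Blueprint for the Jack-o-Lantern backpack cosmetic that drives physics for six color-named pumpkin topper variants.
# On customization load and every animation update, backpack Stage1–Stage6 meta tags set each topper's alpha variable to one or zero.
# From the ref pose, six chained Transform Bone and RigidBody pairs use those alphas to offset Dyn_Master_Pumpkin_Top and simulate each topper's physics asset relative to the pack bone.
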AnimBlueprint /BRCosmetics/Accessories/FORT_Backpacks/Backpack_M_MED_Jack_o_Lantern/Meshes/M_MED_Jack_o_lantern_Pack_AnimBP
target_skeleton=/Game/Accessories/FORT_Backpacks/Backpack_M_MED_Jack_o_lantern/Meshes/M_MED_Jack_o_lantern_Pack_Skeleton
graphs [
 AnimGraph {nodes [
   n1 = Get Orange_Topper() -> (Orange_Topper:double)
   n2 = "Transform (Modify) Bone - Bone: Dyn_Master_Pumpkin_Top"(ComponentPose=Local To Component(LocalPose=Local Space Ref Pose()), Alpha=@n1, Translation="0,0,0", Rotation="0,0,0", Scale="1,1,1") {
    translationMode=BMM_Additive
    translationSpace=BCS_BoneSpace
    alphaBoolBlend {blendOption=Linear}
    boneToModify {boneName=Dyn_Master_Pumpkin_Top}
   }
   n3 = RigidBody(ComponentPose=@n2, Alpha=@n1) {
    bFreezeIncomingPoseOnStart=true
    bOverrideWorldGravity=true
    overridePhysicsAsset=&/M_MED_Jack_o_lantern_OrangeTopper_Pack_Physics
    simulationSpace=BaseBoneSpace
    alphaBoolBlend {blendOption=Linear}
    baseBoneRef {boneName=pack}
    componentAppliedLinearAccClamp=<1000 1000 1000>
    componentLinearAccScale=<0.1 0.1 0.1>
    componentLinearVelScale=<0.1 0.1 0.1>
    overrideWorldGravity {z=-980}
   }
   n4 = Get Blue_Topper() -> (Blue_Topper:double)
   n5 = "Transform (Modify) Bone - Bone: Dyn_Master_Pumpkin_Top"(ComponentPose=@n3, Alpha=@n4, Translation="0,0,0", Rotation="0,0,0", Scale="1,1,1") {
    translationMode=BMM_Additive
    translationSpace=BCS_BoneSpace
    alphaBoolBlend {blendOption=Linear}
    boneToModify {boneName=Dyn_Master_Pumpkin_Top}
   }
   n6 = RigidBody(ComponentPose=@n5, Alpha=@n4) {
    bFreezeIncomingPoseOnStart=true
    bOverrideWorldGravity=true
    overridePhysicsAsset=&/M_MED_Jack_o_lantern_BlueTopper_Pack_Physics
    simulationSpace=BaseBoneSpace
    alphaBoolBlend {blendOption=Linear}
    baseBoneRef {boneName=pack}
    componentAppliedLinearAccClamp=<1000 1000 1000>
    componentLinearAccScale=<0.1 0.1 0.1>
    componentLinearVelScale=<0.1 0.1 0.1>
    overrideWorldGravity {z=-980}
   }
   n7 = Get Green_Topper() -> (Green_Topper:double)
   n8 = "Transform (Modify) Bone - Bone: Dyn_Master_Pumpkin_Top"(ComponentPose=@n6, Alpha=@n7, Translation="0,0,1", Rotation="0,0,0", Scale="1,1,1") {
    translationMode=BMM_Additive
    translationSpace=BCS_BoneSpace
    alphaBoolBlend {blendOption=Linear}
    boneToModify {boneName=Dyn_Master_Pumpkin_Top}
    translation {z=1}
   }
   n9 = RigidBody(ComponentPose=@n8, Alpha=@n7) {
    bFreezeIncomingPoseOnStart=true
    bOverrideWorldGravity=true
    overridePhysicsAsset=&/M_MED_Jack_o_lantern_GreenTopper_Pack_Physics
    simulationSpace=BaseBoneSpace
    alphaBoolBlend {blendOption=Linear}
    baseBoneRef {boneName=pack}
    componentAppliedLinearAccClamp=<1000 1000 1000>
    componentLinearAccScale=<0.1 0.1 0.1>
    componentLinearVelScale=<0.1 0.1 0.1>
    overrideWorldGravity {z=-980}
   }
   n10 = Get Purple_Topper() -> (Purple_Topper:double)
   n11 = "Transform (Modify) Bone - Bone: Dyn_Master_Pumpkin_Top"(ComponentPose=@n9, Alpha=@n10, Translation="0,0,-0.9", Rotation="0,0,0", Scale="1,1,1") {
    rotationMode=BMM_Additive
    rotationSpace=BCS_BoneSpace
    translationMode=BMM_Additive
    translationSpace=BCS_BoneSpace
    alphaBoolBlend {blendOption=Linear}
    boneToModify {boneName=Dyn_Master_Pumpkin_Top}
    translation {z=-0.9}
   }
   n12 = RigidBody(ComponentPose=@n11, Alpha=@n10) {
    bFreezeIncomingPoseOnStart=true
    bOverrideWorldGravity=true
    overridePhysicsAsset=&/M_MED_Jack_o_lantern_PurpleTopper_Pack_Physics
    simulationSpace=BaseBoneSpace
    alphaBoolBlend {blendOption=Linear}
    baseBoneRef {boneName=pack}
    componentAppliedLinearAccClamp=<1000 1000 1000>
    componentLinearAccScale=<0.1 0.1 0.1>
    componentLinearVelScale=<0.1 0.1 0.1>
    overrideWorldGravity {z=-400}
   }
   n13 = Get White_Topper() -> (White_Topper:double)
   n14 = "Transform (Modify) Bone - Bone: Dyn_Master_Pumpkin_Top"(ComponentPose=@n12, Alpha=@n13, Translation="0,0,-0.5", Rotation="0,0,0", Scale="1,1,1") {
    translationMode=BMM_Additive
    translationSpace=BCS_BoneSpace
    alphaBoolBlend {blendOption=Linear}
    boneToModify {boneName=Dyn_Master_Pumpkin_Top}
    translation {z=-0.5}
   }
   n15 = RigidBody(ComponentPose=@n14, Alpha=@n13) {
    bFreezeIncomingPoseOnStart=true
    bOverrideWorldGravity=true
    overridePhysicsAsset=&/M_MED_Jack_o_lantern_WhiteTopper_Pack_Physics
    simulationSpace=BaseBoneSpace
    alphaBoolBlend {blendOption=Linear}
    baseBoneRef {boneName=pack}
    componentAppliedLinearAccClamp=<1000 1000 1000>
    componentLinearAccScale=<0.1 0.1 0.1>
    componentLinearVelScale=<0.1 0.1 0.1>
    overrideWorldGravity {z=-980}
   }
   n16 = Get Red_Topper() -> (Red_Topper:double)
   n17 = "Transform (Modify) Bone - Bone: Dyn_Master_Pumpkin_Top"(ComponentPose=@n15, Alpha=@n16, Translation="0,0,-0.2", Rotation="0,0,0", Scale="1,1,1") {
    translationMode=BMM_Additive
    translationSpace=BCS_BoneSpace
    alphaBoolBlend {blendOption=Linear}
    boneToModify {boneName=Dyn_Master_Pumpkin_Top}
    translation {z=-0.2}
   }
   n18 = RigidBody(ComponentPose=@n17, Alpha=@n16) {
    bClampLinearTranslationLimitToRefPose=true
    bFreezeIncomingPoseOnStart=true
    bOverrideWorldGravity=true
    overridePhysicsAsset=&/M_MED_Jack_o_lantern_RedTopper_Pack_Physics
    simulationSpace=BaseBoneSpace
    alphaBoolBlend {blendOption=Linear}
    baseBoneRef {boneName=pack}
    componentAppliedLinearAccClamp=<1500 1500 1500>
    componentLinearAccScale=<0.5 0.5 0.5>
    componentLinearVelScale=<0.5 0.5 0.5>
    overrideWorldGravity {z=-980}
   }
   Output Pose(Result=Component To Local(ComponentPose=@n18)) {layerGroup=None}
  ]}
 EventGraph {nodes [
   n1 = Event OnCharacterCustomizationLoaded()
   n2 = Sequence!(execute=[@n1.then, @n9.then]) -> (then_0, then_1, then_2:Exec, then_3:Exec, then_4:Exec, then_5:Exec)
   > n3 = Branch(Condition=BackpackItemDefinitionHasAnyMetaTag!(Tags:Array of Gameplay Tag Structures=Make Array("[0]"=(TagName="Cosmetics.Variant.Property.Stage1")), bRequiresEquipped:bool=false) -> (ReturnValue:bool))
   > Set Blue_Topper(Blue_Topper:double=1) -> (Output_Get:double)
   Set Blue_Topper(execute=@n3.else, Blue_Topper=0)
   n4 = Branch(
    execute=@n2.then_1
    Condition=BackpackItemDefinitionHasAnyMetaTag(Tags=Make Array("[0]"=(TagName="Cosmetics.Variant.Property.Stage2")), bRequiresEquipped=false)
   )
   > Set Green_Topper(Green_Topper:double=1) -> (Output_Get:double)
   Set Green_Topper(execute=@n4.else, Green_Topper=0)
   n5 = Branch(
    execute=@n2.then_2
    Condition=BackpackItemDefinitionHasAnyMetaTag(Tags=Make Array("[0]"=(TagName="Cosmetics.Variant.Property.Stage3")), bRequiresEquipped=false)
   )
   > Set Orange_Topper(Orange_Topper:double=1) -> (Output_Get:double)
   Set Orange_Topper(execute=@n5.else, Orange_Topper=0)
   n6 = Branch(
    execute=@n2.then_3
    Condition=BackpackItemDefinitionHasAnyMetaTag(Tags=Make Array("[0]"=(TagName="Cosmetics.Variant.Property.Stage4")), bRequiresEquipped=false)
   )
   > Set Purple_Topper(Purple_Topper:double=1) -> (Output_Get:double)
   Set Purple_Topper(execute=@n6.else, Purple_Topper=0)
   n7 = Branch(
    execute=@n2.then_4
    Condition=BackpackItemDefinitionHasAnyMetaTag(Tags=Make Array("[0]"=(TagName="Cosmetics.Variant.Property.Stage5")), bRequiresEquipped=false)
   )
   > Set Red_Topper(Red_Topper:double=1) -> (Output_Get:double)
   Set Red_Topper(execute=@n7.else, Red_Topper=0)
   n8 = Branch(
    execute=@n2.then_5
    Condition=BackpackItemDefinitionHasAnyMetaTag(Tags=Make Array("[0]"=(TagName="Cosmetics.Variant.Property.Stage6")), bRequiresEquipped=false)
   )
   > Set White_Topper(White_Topper:double=1) -> (Output_Get:double)
   Set White_Topper(execute=@n8.else, White_Topper=0)
   n9 = Event BlueprintUpdateAnimation()
  ]}
]
variables [
 Blue_Topper {type=number}
 Green_Topper {type=number}
 Orange_Topper {type=number}
 Purple_Topper {type=number}
 Red_Topper {type=number}
 White_Topper {type=number}
]
